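# Masked, two-sided subsurface master material for Athena bushes that dissolves foliage near the camera.
# A screen-centered gradient, limited by camera distance, height-lerps against a tiled noise mask to cut opacity.
# World offset combines sine leaf sway, masked hit bounce and cull shrink; flat colors vary by biome and instance.
# Attributes then pass through snow, winter-emissive and Curie landscape functions, with a static switch providing gameplay-bush translucency values.
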Material /Game/Athena/Prototype/Terrain/Athena_Bush_FadeMaster
graph {
 properties {
  bAutomaticallySetUsageInEditor=false
  bTangentSpaceNormal=false
  bUseMaterialAttributes=true
  bUsedWithInstancedStaticMeshes=true
  blendMode=BLEND_Masked
  opacityMaskClipValue=0.5
  pixelDepthOffsetMode=PDOM_Legacy
  shadingModel=MSM_Subsurface
  translucencyLightingMode=TLM_VolumetricPerVertexDirectional
  twoSided=true
  usedShadingModels=Subsurface
 }
 nodes [
  Constant() {desc=.6, r=0.6}
  n1 = StaticBool()
  n2 = SetMaterialAttributes!(
   MaterialAttributes
   Base Color=NamedRerouteUsage()
   Specular=NamedRerouteUsage()
   Roughness=NamedRerouteUsage()
   Emissive Color=NamedRerouteUsage()
   Opacity=NamedRerouteUsage()
   Opacity Mask=NamedRerouteUsage()
   Normal=NamedRerouteUsage()
   World Position Offset=NamedRerouteUsage()
   Subsurface Color=NamedRerouteUsage()
  ) {attributeSetTypes [
    #1
    #2
    #3
    #4
    #5
    #6
    #7
    #8
    #9
   ]}
  n3 = MaterialFunctionCall!(
   IsStaticSnow
   UseAlternateStaticSnowMask
   AlternateStaticSnowMask
   "UseRoad&&OcclusionMask"=@n1
   IsForGrassMeshToPushDown
   AdditionalMaskToRemoveSnow
   IsFallingSnowFX
   UseMinimapCoords
   IncreaseSnowUsingSlope=@n1
   UseSmoothBorder=StaticBool()
   LerpHighM
   TemporalHeightLerpContrast
   IsLobbyIsland
   UseBR-IslandMask
   IsTree
   IsForRenderTarget
   UseSparkle
   UseSnowNormals
   UseAdditionalMaskToRemoveSnow
   IsTangentSpaceInput
   UseVertexNormalsForBasicSnow
   UseTangentSnowNormals=@n1
   Material=@n2
  ) {materialFunction=/Game/Athena/Apollo/Environments/Landscape/Materials/Biome/UniqueBiomeMaterialFunctions/MF_Apollo_Snow_Progression}
  n4 = Constant()
  MaterialFunctionCall!(
   In=MaterialFunctionCall!(In=SetMaterialAttributes!(MaterialAttributes=@n3.Material, Emissive Color=@n4) {attributeSetTypes [#4]}) {materialFunction=/Game/Environments/Props/Winter/Materials/EmissiveTextureWinterTree}.Material Output
   UseHeightFalloff
   OnFireEmissiveOverride=Constant()
   HeightFalloffRadius
   HeightFalloffHardness
   VariationMasksMobile
   In_Grass
   Is_Stylised_WE
  ) {materialFunction=/Game/Packages/Fortress_SharedMaterials/Base_Material_Functions/MF_Curie_Landscape}
  n5 = Sine(None=Multiply(A=Time(), B=ScalarParameter() {defaultValue=0.5, group=LeafAnim, parameterName=LeafAnimationSpeed})) {period=0.25}
  n6 = ScalarParameter() {defaultValue=1, group=LeafAnim, parameterName=LeafAnimationAmount}
  n7 = StaticSwitchParameter(
   True=RotateAboutAxis(
    NormalizedRotationAxis=Constant3Vector() {constant=<rgba 1 0 0 0>}
    RotationAngle=Multiply(A=@n5, B=@n6)
    PivotPoint=ObjectPositionWS()
    Position=WorldPosition().XYZ
   ) {period=28}
   False=Constant()
  ) {defaultValue=true, group=LeafAnim, parameterName=Animate}
  n8 = MaterialFunctionCall!(
   WSDeformationCenter
   Code_DestructionAmount
   Code_UseUvForDamage
   Code_UseHitEffects
   Code_Emissive
   WSDamageRadius
   CrackDamageShader
   Diffuse
   ScalarModDamageCracks
   ScalarOffsetforDamageCracks
   ExposeLerpFunctionForInvertingTheHitDistanceCalc
   InvertDamageFalloff
   CrackWorldPosition
   WorldPosition
   DistanceCullFadeAnimationOffset
  ) {materialFunction=/Game/Packages/jl_tests/Fortnite_HitEffects}
  n9 = WorldPosition() {worldPositionShaderOffset=WPT_ExcludeAllShaderOffsets}
  n10 = Divide(
   A=Subtract(
    A=TransformPosition(None=@n9.XYZ) {transformSourceType=TRANSFORMPOSSOURCE_World}
    B=VectorParameter() {group=WorldPositionOffset, parameterName=Local position of bounce mask, defaultValue=<rgba 0 0 0 1>}.RGB
   )
   B=VectorParameter() {group=WorldPositionOffset, parameterName=Bounce Mask Radius XYZ, defaultValue=<rgba 128 128 128 1>}.RGB
  )
  n11 = Constant()
  n12 = Constant() {r=1}
  n13 = ScalarParameter() {defaultValue=0.5, group=WorldPositionOffset, parameterName=Bounce Mask Hardness}
  n14 = Multiply(A=@n8.Hit WPO, B=OneMinus(None=SphereMask(A=@n10, B=@n11, Radius=@n12, Hardness=@n13) {hardnessPercent=50}))
  n15 = Power(Base=ScalarParameter() {group=WorldPositionOffset, parameterName=DistanceCullFadeAnimationOffset}) {constExponent=5}
  n16 = Add(A=Multiply(A=Add(A=@n7, B=@n14), B=OneMinus(None=@n15)) {desc=remove the animation from the tree as it shrinks}, B=@n8.Shrink WPO) {desc=add the shrinking effect}
  n17 = Constant()
  NamedRerouteDeclaration(None=StaticSwitchParameter(True=@n16, False=QualitySwitch(Default=@n16, Low=@n17)) {defaultValue=true, group=Settings, parameterName="FadeUpClose (Player material)"}) {name=Output-WPO, nodeColor=<rgba 1 0 1 1>}
  NamedRerouteDeclaration(None=Constant()) {name=Emissive, nodeColor=<rgba 1 0 1 1>}
  n18 = TextureSampleParameter2D() {
   group=Textures
   parameterName=MaskTexture
   samplerType=SAMPLERTYPE_Masks
   texture=/Game/Environments/Props/Transylvania/Foliage/Textures/T_TRV_SWP_Plant_04_M
  }
  NamedRerouteDeclaration(None=@n18.R) {name=Mask-R, nodeColor=<rgba 1 0 1 1>}
  NamedRerouteDeclaration(None=@n18.G) {name=Mask-G, nodeColor=<rgba 1 0 1 1>}
  n19 = VectorParameter() {group=Biome, parameterName=BiomeColor, defaultValue=<rgba 0.25 1.6461 2.5 1>}
  n20 = NamedRerouteUsage()
  n21 = LinearInterpolate(
   A=ScalarParameter() {group=Diffuse, parameterName=NewDif-Low}
   B=ScalarParameter() {defaultValue=2, group=Diffuse, parameterName=NewDif-High}
   Alpha=Multiply(A=@n20, B=@n20)
  )
  n22 = Saturate(None=LinearInterpolate(
    A=VectorParameter() {group=Diffuse, parameterName=FlatColor2, defaultValue=<rgba 0.0193 0.05 0 1>}.RGB
    B=VectorParameter() {group=Diffuse, parameterName=FlatColor, defaultValue=<rgba 0.0575 0.105 0 1>}.RGB
    Alpha=Saturate(None=@n21)
   ))
  n23 = LinearInterpolate(
   A=@n22
   B=Saturate(None=Divide(A=@n22, B=@n19.RGB))
   Alpha=MaterialFunctionCall!(UseLandscapeCoords) {materialFunction=/Game/Athena/Apollo/Environments/Landscape/Materials/Biome/CommonMaterials/MF_ApolloTerrainMasks}.CustomBiomeA
  )
  n24 = VectorParameter() {group=Diffuse, parameterName=PIR-Div-A, defaultValue=<rgba 1 1 1 1>}
  n25 = VectorParameter() {group=Diffuse, parameterName=PIR-Div-B, defaultValue=<rgba 0.48 0.8 0.5032 1>}
  n26 = PerInstanceRandom()
  n27 = Saturate(None=Divide(
    A=StaticSwitchParameter(True=@n23, False=@n22) {defaultValue=true, group=Biome, parameterName=UseBiomeMask}
    B=LinearInterpolate(A=@n24.RGB, B=@n25.RGB, Alpha=@n26)
   ))
  NamedRerouteDeclaration(None=@n27) {name=Output-Diffuse, nodeColor=<rgba 1 0 1 1>}
  n28 = MaterialFunctionCall(
   Color_Morning=VectorParameter() {group=Emissive, parameterName=Color_Morning, defaultValue=<rgba 0.2292 0.1066 0.0251 1>}.RGB
   Color_Day=VectorParameter() {group=Emissive, parameterName=Color_Day, defaultValue=<rgba 0.347 0.3958 0.2928 1>}.RGB
   Color_Evening=VectorParameter() {group=Emissive, parameterName=Color_Evening, defaultValue=<rgba 0.2292 0.1066 0.0251 1>}.RGB
   Color_Night=VectorParameter() {group=Emissive, parameterName=Color_Night, defaultValue=<rgba 0.1943 0.2292 0.1502 1>}.RGB
  ) {materialFunction=&Color_Time_of_Day}
  n29 = NamedRerouteUsage()
  NamedRerouteDeclaration(None=StaticSwitchParameter(True=Multiply(A=@n28.Result, B=@n29), False=NamedRerouteUsage()) {desc="For Gameplay Bush ", group=Settings, parameterName=IsTranslucentMaterial}) {name=OutPut-Emissive, nodeColor=<rgba 0.9688 0 1 1>}
  NamedRerouteDeclaration(None=@n27) {name=DiffuseStart, nodeColor=<rgba 1 0 1 1>}
  n30 = TextureSampleParameter2D() {
   group=Textures
   parameterName=NormalMap
   samplerType=SAMPLERTYPE_Normal
   texture=/Game/Environments/Props/Transylvania/Foliage/Textures/T_TRV_SWP_Plant_04_N
  }
  NamedRerouteDeclaration(None=@n30.RGB) {name=TangentNormal, nodeColor=<rgba 1 0 1 1>}
  n31 = NamedRerouteUsage()
  n32 = MaterialFunctionCall!(
   UVs=MaterialFunctionCall(Texture Dimensions=ViewSize()) {materialFunction=&ScreenAlignedUVs}."X 100%, Y 100%"
   CenterPosition
   Radius=Constant() {r=0.55}
  ) {materialFunction=/Engine/Functions/Engine_MaterialFunctions01/Gradient/SphereGradient-2D}
  n33 = Distance(A=WorldPosition().XYZ, B=CameraPositionWS())
  n34 = Constant() {r=130}
  n35 = Saturate(None=Divide(A=Subtract(A=@n33, B=@n34), B=Constant() {r=120}))
  n36 = TextureSample(UVs=TextureCoordinate() {uTiling=3, vTiling=3}) {
   samplerSource=SSM_Wrap_WorldGroupSettings
   samplerType=SAMPLERTYPE_LinearGrayscale
   texture=/Game/Athena/Prototype/Terrain/Textures/T_Athena_Bush_Mask
  }
  n37 = MaterialFunctionCall(Transition Phase=Saturate(None=Subtract(A=@n32.Result 0-1, B=@n35)), Height Texture=OneMinus(None=Multiply(A=@n36.R) {constB=1.3})) {materialFunction=&HeightLerp}
  NamedRerouteDeclaration(None=StaticSwitchParameter(True=Subtract(A=@n31, B=@n37.Lerp Alpha No Contrast), False=@n31) {defaultValue=true, group=Settings, parameterName="FadeUpClose (Player material)"}) {name=Output-OpacityMask, nodeColor=<rgba 1 0 1 1>}
  n38 = TextureCoordinate() {coordinateIndex=1}
  n39 = Add(
   A=Multiply(A=ComponentMask(None=@n38) {r=true}, B=ScalarParameter() {defaultValue=12, group=Specular, parameterName=NewSpec-Grad-M})
   B=ScalarParameter() {defaultValue=-1.5, group=Specular, parameterName=NewSpec-Grad-A}
  )
  n40 = ScalarParameter() {defaultValue=-1, group=Specular, parameterName=NewSpec-Low}
  n41 = ScalarParameter() {defaultValue=1, group=Specular, parameterName=NewSpec-High}
  n42 = NamedRerouteUsage()
  n43 = Multiply(
   A=ScalarParameter() {defaultValue=0.3, group=Specular, parameterName=NewSpec-Mod}
   B=Saturate(None=LinearInterpolate(A=@n40, B=@n41, Alpha=@n42))
  )
  NamedRerouteDeclaration(None=Multiply(A=Saturate(None=@n39), B=Saturate(None=@n43))) {name=Output-Spec, nodeColor=<rgba 1 0 1 1>}
  NamedRerouteDeclaration(None=ScalarParameter() {defaultValue=0.8, group=Roughness, parameterName=FlatRoughness}) {name=Output-Rough, nodeColor=<rgba 1 0 1 1>}
  n44 = VectorParameter() {group=SSS, parameterName=SSS Color, defaultValue=<rgba 0.489 1 0.05 1>}
  n45 = Divide(A=@n44.RGB, B=VectorParameter() {group=Biome, parameterName=BiomeSSColor, defaultValue=<rgba 0.3 1.1159 1.5 1>}.RGB)
  n46 = StaticSwitchParameter(
   True=LinearInterpolate(
    A=@n44.RGB
    B=Saturate(None=@n45)
    Alpha=MaterialFunctionCall() {materialFunction=&MF_ApolloTerrainMasks}.CustomBiomeA
   )
   False=@n44.RGB
  ) {defaultValue=true, group=Biome, parameterName=UseBiomeMask}
  n47 = ScalarParameter() {defaultValue=0.25, group=SSS, parameterName=Amount Of SubSurface Scattering}
  n48 = Constant()
  NamedRerouteDeclaration(None=ShadingPathSwitch(Default=Multiply(A=@n46, B=@n47), Forward=@n48, Mobile=@n48)) {name=Output-SSColor, nodeColor=<rgba 1 0 1 1>}
  n49 = ScalarParameter() {defaultValue=0.381, group=TransDistanceFade, parameterName=Opacity}
  n50 = WorldPosition()
  n51 = Add(
   A=CameraPositionWS()
   B=Multiply(
    A=MaterialFunctionCall() {materialFunction=&CameraDirectionVector}.Result
    B=ScalarParameter() {defaultValue=80, group=TransDistanceFade, parameterName=CameraTranslucentPush}
   )
  )
  n52 = Divide(
   A=Subtract(A=Distance(A=@n50.XYZ, B=@n51), B=ScalarParameter() {defaultValue=50, group=TransDistanceFade, parameterName=TranslucentFadeoffset})
   B=ScalarParameter() {defaultValue=50, group=TransDistanceFade, parameterName=TranslucentFadeDistance}
  )
  n53 = VertexInterpolator(VS=OneMinus(None=Saturate(None=@n52)))
  n54 = MaterialFunctionCall(
   Transition Phase=Saturate(None=Subtract(A=@n49, B=@n53.PS))
   Height Texture=NamedRerouteUsage()
   Contrast=ScalarParameter() {defaultValue=0.15, group=TransDistanceFade, parameterName=ContrastTranslucentLerp}
  ) {materialFunction=&HeightLerp}
  n55 = NamedRerouteUsage()
  n56 = Constant()
  n57 = OneMinus(None=@n47)
  NamedRerouteDeclaration(None=StaticSwitchParameter(True=Multiply(A=@n54.Alpha, B=@n55), False=ShadingPathSwitch(Default=@n57, Forward=@n56, Mobile=@n56)) {desc="For Gameplay Bush ", group=Settings, parameterName=IsTranslucentMaterial}) {name=Output-Opacity, nodeColor=<rgba 1 0 1 1>}
  n58 = Multiply(A=VectorParameter() {group=Normals, parameterName=NormalsMod, defaultValue=<rgba 1 1 0.75 1>}.RGB, B=NamedRerouteUsage())
  n59 = Constant3Vector() {constant=<rgba 0 0 1 0>}
  NamedRerouteDeclaration(None=LinearInterpolate(
    A=Transform(None=@n58)
    B=Transform(None=@n59) {transformSourceType=TRANSFORMSOURCE_Instance}
    Alpha=ScalarParameter() {defaultValue=0.6, desc=.6, group=Normals, parameterName=BlendToInstanceSpaceZ-Normals}
   )) {name=Output-Normals, nodeColor=<rgba 1 0 1 1>}
  ObjectPositionWS()
  StaticBool() {value=true}
  TransformPosition(None=Constant3Vector()) {transformType=TRANSFORMPOSSOURCE_World}
  VertexInterpolator(VS=Transform(None=Constant3Vector()) {transformSourceType=TRANSFORMSOURCE_Local})
 ]
}
